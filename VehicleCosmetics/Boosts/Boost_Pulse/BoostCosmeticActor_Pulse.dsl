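; Cosmetic actor for the Pulse vehicle boost, handling its mesh placement and painted color variants.
; On BeginPlay it moves the skeletal mesh to a fixed offset through the vehicle cosmetics function library.
; When the color is ready it calls the parent, converts the color name to an index, and pushes values from two painted-variant maps into emissive and special-texture material parameters.
; Tick, overlap and construction script add no logic of their own.
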(event EventBeginPlay
  ; Set initial mesh location
  (Class|BPVehicleCosmeticsFunctionLibrary|CosmeticActorSetSkeletalAnimLocation self "-70.000000,0.000000,-4.000000" (Variables|Default|GetSkeletalMesh)))

(event Collision|EventActorBeginOverlap
  (params (OtherActor :type Actor%20Object%20Reference)))

(event EventTick
  (params (DeltaSeconds :type Float%20%28single-precision%29)))

(event Vehicle|Cosmetics|EventOnColorReady
  (params (Color :type Name :default "None") (AdditionalInfo :type Vehicle%20Cosmetics%20Additional%20Parameter%20Infos%20Structure))
  (bind _skeletalmesh (Variables|Default|GetSkeletalMesh))
  (bind _paintedvariant (Variables|Default|GetPaintedVariant))
  (|CallParentFunction Color AdditionalInfo)
  (bind _color_index (Class|BPVehicleCosmeticsFunctions|BPGetNSIndexbyColor Color))
  (Rendering|Material|SetColorParameterValueonMaterials _skeletalmesh "Emissive - Color" (.Value (Utilities|Map|GetKeyValueByIndex _paintedvariant _color_index)))
  (Rendering|Material|SetColorParameterValueonMaterials _skeletalmesh "Special Texture - UVs " (.Value (Utilities|Map|GetKeyValueByIndex _paintedvariant _color_index)))
  (Rendering|Material|SetColorParameterValueonMaterials _skeletalmesh "Special Texture - Tint" (.Value (Utilities|Map|GetKeyValueByIndex (Variables|Default|GetPaintedVariant2) _color_index))))

(fn ConstructionScript
  (|CallParentFunction))
