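# Niagara system for the Pulse vehicle boost cosmetic's cone effect, placed from user car-offset and boost-position variables and tinted through painted-variant color overrides.
# Core_Particles spawns SubUV-animated sprites and a mesh burst at the core, using spawn-group switches to set lifetime and size.
# Core_Particles001 spawns only mesh bursts, spaced evenly around a torus by execution index, oriented by shape vector and alpha-faded.
# Both emitters loop particle lifetimes rather than killing particles.
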NiagaraSystem /VehicleCosmetics/Boosts/Boost_Pulse/FX/NS_Boost_Cone_Pulse
dependencies {
 usedDataInterfaces [
  &NiagaraDataInterfaceCurve
  &NiagaraDataInterfaceSpriteRendererInfo
  &NiagaraDataInterfaceSkeletalMesh
  &NiagaraDataInterfaceArrayFloat3
 ]
 usedDynamicInputs [&SelectVectorFromArray]
 usedModules [
  &self:SystemSpawnScript.SystemScriptSource.SystemScriptGraph.NiagaraNodeAssignment_0.SetVariables_#1
  &self:SystemSpawnScript.SystemScriptSource.SystemScriptGraph.NiagaraNodeAssignment_1.SetVariables_#2
  &self:SystemSpawnScript.SystemScriptSource.SystemScriptGraph.NiagaraNodeAssignment_2.SetVariables_#3
  &SystemState
 ]
}
emitters {
 Core_Particles {
  emitterUpdateScript [
   &EmitterState(Inactive Response=::NewEnumerator1, Life Cycle Mode=::NewEnumerator1, Loop Behavior=::NewEnumerator1, Loop Duration=1, UseLoopDelay=true)
   &SpawnRate(SpawnRate=10, Use Spawn Probability=false)
   &SpawnRate(SpawnGroup=1, Use Spawn Probability=false) {module=SpawnRate001}
   &SpawnBurst_Instantaneous(Spawn Count=5, Spawn Group=2, Use Spawn Probability=false, UseLoopCountLimit=false)
  ]
  particleSpawnScript [
   /SetVariables(Particles.PTag:NiagaraInt32=&SpawnGroup()) {isSetParameters=true, module=SetVariables_#4}
   &VC_PaintedVariantsOverride(
    DefaultColor=<rgba 2.764 5 0.001 1>
    R1 DefaultColor=<rgba 0.2241 1 0.0997 1>
    R2 DefaultColor:LinearColor=<rgba 7.998 15 6.875 0.8696>
    Use R1 Painted Overrides=true
    Use R2 Painted Overrides=true
   )
   &InitializeParticle(
    Color=$Particles.Color
    Lifetime=&MergeHelper_FloatSwitch(
     0=$Emitter.CurrentLoopDuration
     1=$Emitter.CurrentLoopDuration
     2=1
     3=?
     Index=&SpawnGroup()
     Num=::NewEnumerator1
    )
    Mesh Scale Mode=::NewEnumerator3
    Mesh Uniform Scale=&MergeHelper_FloatSwitch(0=0, 1=0, 2=2, 3=?, Index=&SpawnGroup(), Num=::NewEnumerator1)
    Position=<0 35 0>
    Position Mode=::NewEnumerator1
    Position Offset=&AddVectorToPosition(Position=$User.DTCarOffset, Vector=Linked() {linkedVariable {name=User.SingleBoostPos, type=NiagaraPosition}})
    Position Offset Coordinate Space=::Local
    Sprite Rotation Angle=&RandomRangeFloat(Minimum=0, Maximum=1)
    Sprite Rotation Mode=::NewEnumerator1
    Sprite Size Mode=::NewEnumerator3
    Uniform Sprite Size=&MergeHelper_FloatSwitch(
     0=&RandomRangeFloat(Minimum=50, Maximum=100)
     1=150
     2=0
     3=0
     4=0
     Index=&SpawnGroup()
     Num=::NewEnumerator3
    )
    UsePositionOffset=true
    Write Material Random=true
   )
   &AddVelocity(Velocity=<-1 0 0>) {enabled=false}
   &ShapeLocation(
    Handle Radius=0
    Large Radius=&MergeHelper_FloatSwitch(0=0, 1=0, 2=123.1316, 3=?, Index=0, Num=::NewEnumerator1)
    MaskSpawnGroup=false
    Rotation Mode=::NewEnumerator2
    Shape Primitive=::NewEnumerator3
    Torus Distribution Mode:ENiagaraTorusDistributionMode=::NewEnumerator1
    U Position=&ReturnNormalizedExecIndex()
    UseOwnerScale=true
    Yaw / Pitch / Roll=<0 0.08 0.25>
   ) {enabled=false, module=ShapeLocation001}
   &InitialMeshOrientation(Mesh Orientation Mode=::NewEnumerator2, Orientation Vector=$Particles.ShapeLocation001.ShapeVector, Use Rotation Vector=false) {enabled=false}
   &ShapeLocation(
    MaskSpawnGroup=false
    Sphere Distribution=::NewEnumerator1
    Sphere Radius=&MergeHelper_FloatSwitch(0=55, 1=0, 2=0, 3=?, Index=&SpawnGroup(), Num=::NewEnumerator1)
    UseOwnerScale=true
   ) {enabled=false}
  ]
  particleUpdateScript [
   &ParticleState(Kill Particles When Lifetime Has Elapsed=false, Loop Particles Lifetime=true)
   &SolveForcesAndVelocity(Clamp Velocity=false, Limit Acceleration=false)
   &ScaleColor(
    Scale Alpha=&FloatFromCurve(FloatCurve=DataInterface() {propertyValues {Curve {defaultValue=1, keys [{interpMode=RCIM_Cubic, tangentMode=RCTM_User, time=0.5, value=1}, {tangentMode=RCTM_User, time=1, value=1}]}}})
    ScaleA=true
    ScaleRGB=true
   ) {enabled=false}
   &SubUVAnimation(
    Evaluation Type=::NewEnumerator0
    Output Alternate Renderer Binding=true
    "Sprite Renderer "=DataInterface() {propertyValues {SpriteRenderer=&self:Core_Particles_0.NiagaraSpriteRendererProperties_0}}
    SubUV Animation Mode:ENiagara_SubUVLookupModeV2=::NewEnumerator2
    UseEndFrame=false
    UseStartFrame=false
   )
  ]
  rendererClasses [&NiagaraSpriteRendererProperties, &NiagaraMeshRendererProperties]
  renderers [
   {rendererClass=&NiagaraSpriteRendererProperties}
   {rendererClass=&NiagaraSpriteRendererProperties, rendererIndex=1}
   {rendererClass=&NiagaraMeshRendererProperties, rendererIndex=2}
  ]
 }
 Core_Particles001 {
  emitterUpdateScript [
   &EmitterState(Inactive Response=::NewEnumerator1, Life Cycle Mode=::NewEnumerator1, Loop Behavior=::NewEnumerator1, Loop Duration=2, UseLoopDelay=true)
   &SpawnRate(SpawnRate=20, Use Spawn Probability=false) {enabled=false}
   &SpawnRate(SpawnGroup=1, Use Spawn Probability=false) {enabled=false, module=SpawnRate001}
   &SpawnBurst_Instantaneous(Spawn Count=5, Spawn Group=2, Use Spawn Probability=false, UseLoopCountLimit=false)
  ]
  particleSpawnScript [
   /SetVariables(Particles.PTag:NiagaraInt32=&SpawnGroup()) {isSetParameters=true, module=SetVariables_#4}
   &VC_PaintedVariantsOverride(
    Black=<rgba 0 0 0 40>
    BurntSienna=<rgba 0.5573 0.1721 0.0687 50>
    Cobalt=<rgba 0.6667 0 4 50>
    Crimson=<rgba 5 0 0.1355 50>
    DefaultColor=<rgba 2.764 5 0.001 1>
    ForestGreen=<rgba 0 2 0.1011 50>
    Gold:LinearColor=<rgba 1 1 1 40>
    Grey=<rgba 0.5 0.5 0.5 50>
    Lime=<rgba 0.2943 0.5885 0 50>
    Orange=<rgba 0.6724 0.2346 0 50>
    Pink=<rgba 0.5052 0.0605 0.3435 50>
    Purple=<rgba 0.2789 0 0.6795 50>
    R1 DefaultColor=<rgba 0.2241 1 0.0997 1>
    R2 DefaultColor:LinearColor=<rgba 7.998 15 6.875 0.8696>
    Saffron=<rgba 0.9 0.69 0 50>
    SkyBlue=<rgba 0 0.357 0.5365 50>
    TitaniumWhite=<rgba 1 1 1 50>
    Use R1 Painted Overrides=true
    Use R2 Painted Overrides=true
   )
   &InitializeParticle(
    Color=$Particles.Color
    Lifetime=&MergeHelper_FloatSwitch(
     0=&RandomRangeFloat(Minimum=0.2, Maximum=0.5)
     1=5
     2=$Emitter.CurrentLoopDuration
     3=?
     Index=&SpawnGroup()
     Num=::NewEnumerator1
    )
    Mesh Scale Mode=::NewEnumerator3
    Mesh Uniform Scale=&MergeHelper_FloatSwitch(0=0, 1=0, 2=3, 3=?, Index=&SpawnGroup(), Num=::NewEnumerator1)
    Position=<0 35 0>
    Position Mode=::NewEnumerator1
    Position Offset=&AddVectorToPosition(Position=$User.DTCarOffset, Vector=Linked() {linkedVariable {name=User.SingleBoostPos, type=NiagaraPosition}})
    Position Offset Coordinate Space=::Local
    Sprite Rotation Angle=&RandomRangeFloat(Minimum=0, Maximum=1)
    Sprite Rotation Mode=::NewEnumerator1
    Sprite Size Mode=::NewEnumerator3
    Uniform Sprite Size=&MergeHelper_FloatSwitch(
     0=&RandomRangeFloat(Minimum=50, Maximum=100)
     1=150
     2=0
     3=0
     4=0
     Index=&SpawnGroup()
     Num=::NewEnumerator3
    )
    UsePositionOffset=true
    Write Material Random=true
   )
   &AddVelocity(Velocity=<-1 0 0>) {enabled=false}
   &ShapeLocation(
    Handle Radius=0
    Large Radius=120
    MaskSpawnGroup=false
    Rotation Mode=::NewEnumerator2
    Shape Primitive=::NewEnumerator3
    Torus Distribution Mode:ENiagaraTorusDistributionMode=::NewEnumerator1
    U Position=&ReturnNormalizedExecIndex()
    UseOwnerScale=true
    Yaw / Pitch / Roll=<0 0.14 0.25>
   ) {module=ShapeLocation001}
   &InitialMeshOrientation(Mesh Orientation Mode=::NewEnumerator2, Orientation Vector=$Particles.ShapeLocation001.ShapeVector, Use Rotation Vector=false)
   &ShapeLocation(
    MaskSpawnGroup=false
    Sphere Distribution=::NewEnumerator1
    Sphere Radius=&MergeHelper_FloatSwitch(0=55, 1=0, 2=0, 3=?, Index=&SpawnGroup(), Num=::NewEnumerator1)
    UseOwnerScale=true
   ) {enabled=false}
  ]
  particleUpdateScript [
   &ParticleState(Kill Particles When Lifetime Has Elapsed=false, Loop Particles Lifetime=true)
   &SolveForcesAndVelocity(Clamp Velocity=false, Limit Acceleration=false)
   &ScaleColor(
    Scale Alpha=&FloatFromCurve(FloatCurve=DataInterface() {propertyValues {Curve {defaultValue=1, keys [{interpMode=RCIM_Cubic, tangentMode=RCTM_User, time=0.5, value=1}, {tangentMode=RCTM_User, time=1, value=0}]}}})
    ScaleA=true
    ScaleRGB=true
   )
   &SubUVAnimation(
    Evaluation Type=::NewEnumerator0
    Output Alternate Renderer Binding=true
    "Sprite Renderer "=DataInterface() {propertyValues {SpriteRenderer=&self:Core_Particles001_0.NiagaraSpriteRendererProperties_0}}
    SubUV Animation Mode:ENiagara_SubUVLookupModeV2=::NewEnumerator2
    UseEndFrame=false
    UseStartFrame=false
   ) {enabled=false}
  ]
  rendererClasses [&NiagaraSpriteRendererProperties, &NiagaraMeshRendererProperties]
  renderers [
   {rendererClass=&NiagaraSpriteRendererProperties}
   {rendererClass=&NiagaraSpriteRendererProperties, rendererIndex=1}
   {rendererClass=&NiagaraMeshRendererProperties, rendererIndex=2}
  ]
 }
}
summary {
 systemName=NS_Boost_Cone_Pulse
 emitters [
  {emitterName=Core_Particles, rendererClasses [&NiagaraSpriteRendererProperties, &NiagaraMeshRendererProperties]}
  {
   emitterName=Core_Particles001
   rendererClasses [&NiagaraSpriteRendererProperties, &NiagaraMeshRendererProperties]
  }
 ]
 userVariables [
  User.bIsFrontendPreview {type=NiagaraBool}
  User.bIsPainted {type=NiagaraBool, defaultValue=true}
  User.bPreviewVehicleSpeed {type=NiagaraBool}
  User.ColorVariantID {type=NiagaraInt32}
  User.DTCarOffset {type=NiagaraPosition, description="", defaultValue {}}
  User.GameModeID {type=NiagaraInt32}
  User.PaintedColor {type=LinearColor, description="", defaultValue {struct=LinearColor, value=<rgba 1 1 1 1>}}
  User.PaintedScalar {type=NiagaraFloat, defaultValue=1, description=""}
  User.PortCount {type=NiagaraInt32, defaultValue=1}
  User.SingleBoostPos {type=NiagaraPosition, description="", defaultValue {}}
 ]
}
