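# Niagara effect for the Pulse vehicle boost cosmetic, emitting a green-by-default trail behind the car, placed from the DTCarOffset user position.
# A fast backward ribbon joins emitters for gravity-pulled sparks, sprite and mesh flashes, and cylinder-shaped sprite and mesh bursts.
# Waveforms vary spawn counts and rates, and spawn-group switches give multiple particle types per emitter.
# Painted-variant overrides recolor each active layer; one SubUV sprite emitter is disabled.
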NiagaraSystem /VehicleCosmetics/Boosts/Boost_Pulse/FX/NS_Pulse
dependencies {
 usedDataInterfaces [
  &NiagaraDataInterfaceVectorCurve
  &NiagaraDataInterfaceCurve
  &NiagaraDataInterfaceSpriteRendererInfo
  &NiagaraDataInterfaceSkeletalMesh
 ]
 usedDynamicInputs []
 usedModules [
  &self:SystemSpawnScript.SystemScriptSource.SystemScriptGraph.NiagaraNodeAssignment_0.SetVariables_#1
  &SystemState
 ]
}
emitters {
 NE_VC_SingleBoost_Ribbon_Parent {
  emitterUpdateScript [
   &EmitterState()
   &VC_FilteredSocketCount()
   &VC_RateQualityHandler()
   &VC_SpawnRibbon_v1(SpawnRate=$Emitter.Spawn Rate)
  ]
  particleSpawnScript [
   &VC_PaintedVariantsOverride(
    Black=<rgba 0 0 0 1>
    BurntSienna=<rgba 0.5 0.2443 0.0616 0.5>
    Cobalt=<rgba 0.1882 0 1 0.5>
    Crimson=<rgba 0.4635 0 0.0126 0.5>
    DefaultColor=<rgba 0.0853 0.2083 0.0347 1>
    ForestGreen=<rgba 0.1024 0.3073 0 1>
    Grey=<rgba 0.2031 0.2031 0.2031 0.5>
    Lime=<rgba 0.2943 0.5885 0 0.5>
    Orange=<rgba 0.3281 0.1145 0 0.5>
    Pink=<rgba 0.5052 0.0289 0.332 0.5>
    Purple=<rgba 0.2789 0 0.6795 0.5>
    R1 Black:LinearColor=<rgba 0 0 0 3>
    R1 BurntSienna:LinearColor=<rgba 0.8 0.3909 0.0986 2>
    R1 Cobalt:LinearColor=<rgba 0.1979 0.8194 1 2>
    R1 Crimson:LinearColor=<rgba 1 0.3802 0.397 2.5>
    R1 DefaultColor=<rgba 1.0404 1.5 0.8516 1>
    R1 ForestGreen:LinearColor=<rgba 0.5 1 0 2>
    R1 Grey:LinearColor=<rgba 1 1 1 2>
    R1 Lime:LinearColor=<rgba 0.5 1 0 2>
    R1 Orange:LinearColor=<rgba 3 1.0464 0 2>
    R1 Pink:LinearColor=<rgba 2 1.0282 1.3385 2>
    R1 Purple:LinearColor=<rgba 1.6417 0 4 2>
    R1 Saffron:LinearColor=<rgba 1.5 1.5 0 2>
    R1 SkyBlue:LinearColor=<rgba 0.8828 1.2935 1.5 2>
    R1 TitaniumWhite:LinearColor=<rgba 1 1 1 2>
    R2 Black:LinearColor=<rgba 0 0 0 3>
    R2 BurntSienna:LinearColor=<rgba 1 0.4886 0.1233 5>
    R2 Cobalt:LinearColor=<rgba 0 1.7767 2 1>
    R2 Crimson:LinearColor=<rgba 2 0.4062 0.4494 1>
    R2 ForestGreen:LinearColor=<rgba 0 2 0.1011 1>
    R2 Grey:LinearColor=<rgba 2 2 2 1>
    R2 Lime:LinearColor=<rgba 1 2 0 1>
    R2 Orange:LinearColor=<rgba 2 0.6976 0 1>
    R2 Pink:LinearColor=<rgba 2 0.2396 1.3598 1>
    R2 Purple:LinearColor=<rgba 0.8208 0 2 1>
    R2 Saffron:LinearColor=<rgba 2 2 0 1>
    R2 SkyBlue:LinearColor=<rgba 0 1.3308 2 1>
    R2 TitaniumWhite:LinearColor=<rgba 2 2 2 1>
    Saffron=<rgba 0.9 0.9 0 0.2>
    SkyBlue=<rgba 0 0.357 0.5365 0.5>
    TitaniumWhite=<rgba 0.75 0.75 0.75 0.2>
    Use R1 Painted Overrides=true
    Use R2 Painted Overrides=true
   )
   &VC_InitializeRibbon_v1(
    Color=$Particles.Color
    Custom Offset=true
    Lifetime=0.35
    Local Vector Offset=&AddVectorToPosition(Position=$User.DTCarOffset, Vector=<-275 0 25>)
    Offset Amount=1
    Ribbon Width=125
   )
   &AddVelocity(Velocity=<-4000 0 0>) {module=AddVelocity001}
  ]
  particleUpdateScript [
   &ParticleState()
   &SolveForcesAndVelocity(Clamp Velocity=false, Limit Acceleration=false)
   &VC_UpdateRibbon_v1()
   /SetVariables() {isSetParameters=true}
   /SetVariables() {isSetParameters=true}
  ]
  rendererClasses [&NiagaraRibbonRendererProperties]
  renderers [
   {rendererClass=&NiagaraRibbonRendererProperties}
   {rendererClass=&NiagaraRibbonRendererProperties, rendererIndex=1}
   {rendererClass=&NiagaraRibbonRendererProperties, rendererIndex=2}
  ]
 }
 NE_VC_SingleBoost_Sprite_Parent {
  emitterUpdateScript [
   &EmitterState(Inactive Response=::NewEnumerator1, Life Cycle Mode=::NewEnumerator1, Loop Delay=0.25, UseLoopDelay=true)
   &SpawnBurst_Instantaneous(
    Spawn Count=&MakeIntFromFloat(Float=&Waveform(Global Amplitude Scale=48, "Waveform [1]"=::NewEnumerator0, "[1] Frequency"=0.5, "[1] Amplitude Min/Max"=<0 1>))
    Use Spawn Probability=false
    UseLoopCountLimit=false
   )
   &SpawnRate(SpawnGroup=1, Use Spawn Probability=false)
  ]
  particleSpawnScript [
   /SetVariables(Particles.VisibilityTag=$Emitter.SpawnGroup) {isSetParameters=true, module=SetVariables_#2}
   &VC_PaintedVariantsOverride(DefaultColor=<rgba 0.6666 3 0 1>, R1 DefaultColor=<rgba 0.2241 1 0.0997 1>, Use R1 Painted Overrides=true)
   &InitializeParticle(
    Color=$Particles.Color
    Lifetime=&MergeHelper_FloatSwitch(
     0=&RandomRangeFloat(Minimum=0.3, Maximum=0.5)
     1=5
     2=0
     3=0
     4=0
     Index=$Emitter.SpawnGroup
     Num=::NewEnumerator3
    )
    Mass Mode=::NewEnumerator1
    Position Offset=&AddVectorToPosition(
     Position=$User.DTCarOffset
     Vector=&MergeHelper_VectorSwitch(0=<-235 0 30>, 1=<-240 0 30>, 2=<0 0 0>, 3=<0 0 0>, 4=<0 0 0>, Index=$Emitter.SpawnGroup, Num=::NewEnumerator3)
    )
    Position Offset Coordinate Space=::Local
    Sprite Rotation Mode=::NewEnumerator2
    Sprite Size=&MergeHelper_Vector2Switch(
     0=&RandomRangeVector2D(Minimum=<3 20>, Maximum=<5 30>)
     1=<30 30>
     2=<0 0>
     3=<0 0>
     4=<0 0>
     Index=$Emitter.SpawnGroup
     Num=::NewEnumerator3
    )
    Sprite Size Mode=::NewEnumerator4
    UsePositionOffset=true
    Write Material Random=true
   )
   &ShapeLocation(MaskSpawnGroup=true, Offset Mode=::NewEnumerator0, Override Seed=false, Sphere Radius=5, UseOwnerScale=true)
   &AddVelocity(
    Module.Use Falloff Distance=false
    Origin Offset:NiagaraPosition=$Particles.ShapeLocation.ShapePosition
    Override Seed=false
    Velocity Mode=::NewEnumerator1
    Velocity Speed=&MergeHelper_FloatSwitch(
     0=&RandomRangeFloat(Minimum=350, Maximum=400)
     1=0
     2=0
     3=0
     4=0
     Index=$Emitter.SpawnGroup
     Num=::NewEnumerator3
    )
   )
  ]
  particleUpdateScript [
   &ParticleState()
   &SpriteRotationRate()
   &ScaleSpriteSize(
    Scale Sprite Size Mode=::NewEnumerator0
    Uniform Scale Factor=&MergeHelper_FloatSwitch(
     0=&FloatFromCurve(FloatCurve=DataInterface() {propertyValues {Curve {keys [<at 0 1 RCIM_Cubic>,<at 1 0>]}}})
     1=&FloatFromCurve(FloatCurve=DataInterface() {propertyValues {Curve {keys [{arriveTangent=20.8082, interpMode=RCIM_Cubic, leaveTangent=20.8082, tangentMode=RCTM_User, time=0, value=1}, <at 1 1>]}}})
     2=0
     3=0
     4=0
     Index=&SpawnGroup()
     Num=::NewEnumerator3
    )
   )
   &GravityForce(Gravity=&MergeHelper_VectorSwitch(
     0=&RandomRangeVector(Minimum=<0 0 -980>, Maximum=<0 0 -2000>)
     1=<0 0 0>
     2=<0 0 0>
     3=<0 0 0>
     4=<0 0 0>
     Index=&SpawnGroup()
     Num=::NewEnumerator3
    ))
   &SolveForcesAndVelocity(Clamp Velocity=false, Limit Acceleration=false)
   &ScaleColor(
    Scale Alpha=&FloatFromCurve(FloatCurve=DataInterface() {propertyValues {Curve {defaultValue=1, keys [{interpMode=RCIM_Cubic, tangentMode=RCTM_User, time=0.5, value=1}, {tangentMode=RCTM_User, time=1, value=0}]}}})
    ScaleA=true
    ScaleRGB=true
   )
  ]
  rendererClasses [&NiagaraSpriteRendererProperties]
  renderers [
   {rendererClass=&NiagaraSpriteRendererProperties}
   {rendererClass=&NiagaraSpriteRendererProperties, rendererIndex=1}
  ]
 }
 NE_VC_SingleBoost_Sprite_Parent001 {
  emitterUpdateScript [
   &EmitterState(Inactive Response=::NewEnumerator1, Life Cycle Mode=::NewEnumerator1, Loop Delay=0.25, UseLoopDelay=true)
   &SpawnRate(
    Spawn Probability=0.5
    SpawnRate=&Waveform(Global Amplitude Scale=2, "Waveform [1]"=::NewEnumerator0, "[1] Frequency"=1, "[1] Amplitude Min/Max"=<0 1>)
    Use Spawn Probability=true
   )
   &SpawnRate(
    Spawn Probability=0.25
    SpawnGroup=1
    SpawnRate=&Waveform(Global Amplitude Scale=6, "Waveform [1]"=::NewEnumerator0, "[1] Frequency"=1, "[1] Amplitude Min/Max"=<0 1>, "[1] Scale"=1.75)
    Use Spawn Probability=true
   ) {module=SpawnRate001}
   &SpawnRate(
    Spawn Probability=0.5
    SpawnGroup=2
    SpawnRate=&Waveform(Global Amplitude Scale=2, "Waveform [1]"=::NewEnumerator0, "[1] Frequency"=1, "[1] Amplitude Min/Max"=<0 1>)
    Use Spawn Probability=true
   ) {module=SpawnRate002}
   &SpawnRate(
    Spawn Probability=0.5
    SpawnGroup=3
    SpawnRate=&Waveform(Global Amplitude Scale=2, "Waveform [1]"=::NewEnumerator0, "[1] Frequency"=0.5, "[1] Amplitude Min/Max"=<0 1>)
    Use Spawn Probability=true
   ) {module=SpawnRate003}
   &SpawnRate(SpawnGroup=4, SpawnRate=15, Use Spawn Probability=false) {module=SpawnRate004}
  ]
  particleSpawnScript [
   /SetVariables(Particles.MTag:NiagaraInt32=&SpawnGroup()) {isSetParameters=true}
   &VC_PaintedVariantsOverride(
    Black=<rgba 0 0 0 1>
    BurntSienna=<rgba 0.5208 0.0694 0 1>
    Cobalt=<rgba 0.0242 0 1 1>
    Crimson=<rgba 2 0.2708 0.3177 1>
    DefaultColor=<rgba 0.6826 2 0.5833 1>
    ForestGreen=<rgba 0.3542 2 0.4373 1>
    Grey=<rgba 0.1823 0.1823 0.1823 1>
    Orange=<rgba 1 0.4743 0.1927 2>
    Pink=<rgba 1 0.25 0.7273 1>
    Purple=<rgba 0.8208 0 2 1>
    R1 DefaultColor=<rgba 0 0 0 1>
    Use R1 Painted Overrides=true
   )
   &InitializeParticle(
    Color=$Particles.Color
    Lifetime=&MergeHelper_FloatSwitch(
     0=&RandomRangeFloat(Minimum=0.1, Maximum=0.2)
     1=0.5
     2=0.4
     3=0.1
     4=0.4
     Index=&SpawnGroup()
     Num=::NewEnumerator3
    )
    Mesh Scale Mode=::NewEnumerator3
    Mesh Uniform Scale=1.25
    Position Offset=&AddVectorToPosition(Position=$User.DTCarOffset, Vector=<-250 0 25>)
    Position Offset Coordinate Space=::Local
    Sprite Rotation Mode=::NewEnumerator2
    Sprite Size Mode=::NewEnumerator3
    Uniform Sprite Size=&MergeHelper_FloatSwitch(
     0=&RandomRangeFloat(Minimum=50, Maximum=100)
     1=1.25
     2=&RandomRangeFloat(Minimum=50, Maximum=100)
     3=250
     4=&RandomRangeFloat(Minimum=5, Maximum=10)
     Index=$Emitter.SpawnGroup
     Num=::NewEnumerator3
    )
    UsePositionOffset=true
    Write Material Random=true
   )
   &AddVelocity(Velocity=&MergeHelper_VectorSwitch(
     0=<0 0 0>
     1=&RandomRangeVector(Minimum=<-100 0 0>, Maximum=<-125 0 0>)
     2=<0 0 0>
     3=<0 0 0>
     4=<-2000 0 0>
     Index=&SpawnGroup()
     Num=::NewEnumerator3
    ))
   &InitialMeshOrientation(
    Mesh Orientation Mode=::NewEnumerator2
    Orientation Vector=&MergeHelper_VectorSwitch(0=<0 1 0>, 1=<0 1 0>, 2=<1 0 0>, 3=?, Index=&SpawnGroup(), Num=::NewEnumerator1)
    Use Rotation Vector=false
   )
   &ShapeLocation(
    MaskSpawnGroup=false
    Override Seed=false
    Sphere Radius=&MergeHelper_FloatSwitch(0=0, 1=0, 2=0, 3=0, 4=25, Index=&SpawnGroup(), Num=::NewEnumerator3)
    UseOwnerScale=true
   )
  ]
  particleUpdateScript [
   &ParticleState()
   &UpdateMeshOrientation(Rotation Rate=&MergeHelper_FloatSwitch(0=0, 1=0, 2=0, 3=0, 4=0, Index=&SpawnGroup(), Num=::NewEnumerator3))
   &ScaleMeshSize(Scale Factor=&VectorFromCurve(VectorCurve=DataInterface() {propertyValues {
       XCurve {
        defaultValue=1
        keys [{interpMode=RCIM_Cubic, tangentMode=RCTM_User, time=0, value=0}, {interpMode=RCIM_Cubic, tangentMode=RCTM_User, time=1, value=1}]
       }
       YCurve {
        defaultValue=1
        keys [{interpMode=RCIM_Cubic, tangentMode=RCTM_User, time=0, value=0}, {interpMode=RCIM_Cubic, tangentMode=RCTM_User, time=1, value=1}]
       }
       ZCurve {
        defaultValue=1
        keys [{interpMode=RCIM_Cubic, tangentMode=RCTM_User, time=0, value=0}, {interpMode=RCIM_Cubic, tangentMode=RCTM_User, time=1, value=1}]
       }
      }}))
   &ScaleSpriteSize(
    Scale Sprite Size Mode=::NewEnumerator0
    Uniform Scale Factor=&MergeHelper_FloatSwitch(
     0=&FloatFromCurve(FloatCurve=DataInterface() {propertyValues {Curve {keys [{tangentMode=RCTM_User, time=0, value=0}, {tangentMode=RCTM_User, time=1, value=1}]}}})
     1=&FloatFromCurve(FloatCurve=DataInterface() {propertyValues {Curve {keys [{tangentMode=RCTM_User, time=0, value=0}, {tangentMode=RCTM_User, time=1, value=1}]}}})
     2=&FloatFromCurve(FloatCurve=DataInterface() {propertyValues {Curve {keys [{tangentMode=RCTM_User, time=0, value=0}, {tangentMode=RCTM_User, time=1, value=1}]}}})
     3=&FloatFromCurve(FloatCurve=DataInterface() {propertyValues {Curve {keys [{tangentMode=RCTM_User, time=0, value=0}, {tangentMode=RCTM_User, time=1, value=1}]}}})
     4=1.25
     Index=&SpawnGroup()
     Num=::NewEnumerator3
    )
   )
   &SolveForcesAndVelocity(Clamp Velocity=false, Limit Acceleration=false)
   &ScaleColor(
    Scale Alpha=&FloatFromCurve(FloatCurve=DataInterface() {propertyValues {Curve {defaultValue=1, keys [{interpMode=RCIM_Cubic, tangentMode=RCTM_User, time=0.5, value=1}, {tangentMode=RCTM_User, time=1, value=0}]}}})
    ScaleA=true
    ScaleRGB=true
   )
  ]
  rendererClasses [&NiagaraSpriteRendererProperties, &NiagaraMeshRendererProperties]
  renderers [
   {rendererClass=&NiagaraSpriteRendererProperties}
   {rendererClass=&NiagaraSpriteRendererProperties, rendererIndex=1}
   {rendererClass=&NiagaraMeshRendererProperties, rendererIndex=2}
   {rendererClass=&NiagaraSpriteRendererProperties, rendererIndex=3}
   {rendererClass=&NiagaraSpriteRendererProperties, rendererIndex=4}
  ]
 }
 NE_VC_SingleBoost_Sprite_Parent002 {
  bEnabled=false
  emitterUpdateScript [
   &EmitterState(Inactive Response=::NewEnumerator1, Life Cycle Mode=::NewEnumerator1, Loop Delay=0.25, UseLoopDelay=true)
   &SpawnRate(SpawnGroup=1, SpawnRate=5, Use Spawn Probability=false)
  ]
  particleSpawnScript [
   /SetVariables(Particles.VisibilityTag=$Emitter.SpawnGroup) {isSetParameters=true, module=SetVariables_#2}
   &VC_PaintedVariantsOverride(DefaultColor=<rgba 0.6666 3 0 1>, R1 DefaultColor=<rgba 0.2241 1 0.0997 1>, Use R1 Painted Overrides=true)
   &InitializeParticle(
    Color=$Particles.Color
    Mass Mode=::NewEnumerator1
    Position Offset=&AddVectorToPosition(Position=$User.DTCarOffset, Vector=<0 0 0>)
    Position Offset Coordinate Space=::Local
    Sprite Rotation Mode=::NewEnumerator2
    Sprite Size=&MergeHelper_Vector2Switch(
     0=&RandomRangeVector2D(Minimum=<3 20>, Maximum=<5 30>)
     1=<30 30>
     2=<0 0>
     3=<0 0>
     4=<0 0>
     Index=$Emitter.SpawnGroup
     Num=::NewEnumerator3
    )
    Sprite Size Mode=::NewEnumerator4
    UsePositionOffset=true
    Write Material Random=true
   )
   &ShapeLocation(MaskSpawnGroup=true, Offset Mode=::NewEnumerator0, Override Seed=false, Sphere Radius=5, UseOwnerScale=true)
   &AddVelocity(
    Module.Use Falloff Distance=false
    Origin Offset:NiagaraPosition=$Particles.ShapeLocation.ShapePosition
    Override Seed=false
    Velocity Mode=::NewEnumerator1
    Velocity Speed=&MergeHelper_FloatSwitch(
     0=&RandomRangeFloat(Minimum=350, Maximum=400)
     1=0
     2=0
     3=0
     4=0
     Index=$Emitter.SpawnGroup
     Num=::NewEnumerator3
    )
   )
  ]
  particleUpdateScript [
   &ParticleState()
   &SpriteRotationRate()
   &ScaleSpriteSize(Scale Sprite Size Mode=::NewEnumerator0)
   &SolveForcesAndVelocity(Clamp Velocity=false, Limit Acceleration=false)
   &SubUVAnimation(
    Evaluation Type=::NewEnumerator0
    Output Alternate Renderer Binding=true
    "Sprite Renderer "=DataInterface() {propertyValues {SpriteRenderer=&self:NE_VC_SingleBoost_Sprite_Parent002_0.NiagaraSpriteRendererProperties_0}}
    SubUV Animation Mode:ENiagara_SubUVLookupModeV2=::NewEnumerator2
    UseEndFrame=false
    UseStartFrame=false
   )
  ]
  rendererClasses [&NiagaraSpriteRendererProperties]
  renderers [{rendererClass=&NiagaraSpriteRendererProperties}]
 }
 NE_VC_SingleBoost_Sprite_Parent003 {
  emitterUpdateScript [
   &EmitterState(Inactive Response=::NewEnumerator1, Life Cycle Mode=::NewEnumerator1, UseLoopDelay=true)
   &SpawnBurst_Instantaneous(
    Spawn Count=&MakeIntFromFloat(Float=&Waveform(Global Amplitude Scale=96, "Waveform [1]"=::NewEnumerator0, "[1] Phase Scale"=4, "[1] Frequency"=0.1, "[1] Amplitude Min/Max"=<0 1>))
    Use Spawn Probability=false
    UseLoopCountLimit=false
   )
   &SpawnBurst_Instantaneous(
    Spawn Count=&MakeIntFromFloat(Float=&Waveform(
      Global Amplitude Scale=12
      "Waveform [1]"=::NewEnumerator0
      "[1] Phase Scale"=4
      "[1] Frequency"=1
      "[1] Amplitude Min/Max"=<0 1>
      "[1] Scale"=1.75
     ))
    Spawn Group=4
    Use Spawn Probability=false
    UseLoopCountLimit=false
   ) {module=SpawnBurst_Instantaneous001}
   &SpawnRate(
    SpawnRate=&Waveform(Global Amplitude Scale=12, "Waveform [1]"=::NewEnumerator0, "[1] Frequency"=1, "[1] Amplitude Min/Max"=<0 0.5>, "[1] Scale"=2)
    Use Spawn Probability=false
   )
   &SpawnRate(
    SpawnGroup=1
    SpawnRate=&Waveform(Global Amplitude Scale=36, "Waveform [1]"=::NewEnumerator0, "[1] Frequency"=1, "[1] Amplitude Min/Max"=<0 0.5>, "[1] Scale"=2)
    Use Spawn Probability=false
   ) {module=SpawnRate001}
  ]
  particleSpawnScript [
   /SetVariables(Particles.PTag:NiagaraInt32=&SpawnGroup()) {isSetParameters=true}
   &VC_PaintedVariantsOverride(
    DefaultColor=<rgba 2.764 5 0.001 1>
    R1 DefaultColor=<rgba 0.4482 2 0.1993 1>
    R2 DefaultColor:LinearColor=<rgba 7.998 15 6.875 0.8696>
    Use R1 Painted Overrides=true
    Use R2 Painted Overrides=true
   )
   &InitializeParticle(
    Color=$Particles.Color
    Lifetime=&MergeHelper_FloatSwitch(
     0=&RandomRangeFloat(Minimum=0.25, Maximum=0.4)
     1=&RandomRangeFloat(Minimum=0.15, Maximum=1)
     2=5
     3=5
     4=1
     Index=&SpawnGroup()
     Num=::NewEnumerator3
    )
    Mesh Scale Mode=::NewEnumerator3
    Mesh Uniform Scale=&MergeHelper_FloatSwitch(0=0, 1=1.75, 2=0.75, 3=1.75, 4=1.5, Index=&SpawnGroup(), Num=::NewEnumerator3)
    Position Offset=&AddVectorToPosition(
     Position=$User.DTCarOffset
     Vector=&MergeHelper_VectorSwitch(
      0=<-235 0 25>
      1=<-200 0 25>
      2=<-240 0 25>
      3=<-202 0 25>
      4=<-200 0 25>
      Index=&SpawnGroup()
      Num=::NewEnumerator3
     )
    )
    Position Offset Coordinate Space=::Local
    Sprite Rotation Mode=::NewEnumerator2
    Sprite Size Mode=::NewEnumerator3
    Uniform Sprite Size=&MergeHelper_FloatSwitch(
     0=&RandomRangeFloat(Minimum=75, Maximum=125)
     1=&RandomRangeFloat(Minimum=15, Maximum=25)
     2=0
     3=0
     4=0
     Index=&SpawnGroup()
     Num=::NewEnumerator3
    )
    UsePositionOffset=true
    Write Material Random=true
   )
   &AddVelocity(Velocity=<-1 0 0>)
   &ShapeLocation(
    Cylinder Height=&MergeHelper_FloatSwitch(0=25, 1=50, 2=0, 3=0, 4=0, Index=&SpawnGroup(), Num=::NewEnumerator3)
    Cylinder Radius=&MergeHelper_FloatSwitch(0=25, 1=30, 2=0, 3=0, 4=0, Index=&SpawnGroup(), Num=::NewEnumerator3)
    Enable Lathe Profile=true
    MaskSpawnGroup=false
    Override Seed=false
    Shape Primitive=::NewEnumerator1
    Surface Only=true
    UseOwnerScale=true
   )
   &InitialMeshOrientation(
    Mesh Orientation Mode=::NewEnumerator2
    Orientation Vector=&MergeHelper_VectorSwitch(0=<1 0 0>, 1=<1 0 0>, 2=<0 -1 0>, 3=<0 0 0>, 4=<1 0 0>, Index=&SpawnGroup(), Num=::NewEnumerator3)
    Use Rotation Vector=false
   )
  ]
  particleUpdateScript [
   &ParticleState()
   &UpdateMeshOrientation(Rotation Vector=&MergeHelper_VectorSwitch(
     0=<0 0 0>
     1=<0.5 0 0>
     2=<0 -0.5 0>
     3=<0.5 0 0>
     4=&RandomRangeVector(Minimum=<0 0 0>, Maximum=<0 0 0>)
     Index=&SpawnGroup()
     Num=::NewEnumerator3
    ))
   &SolveForcesAndVelocity(Clamp Velocity=false, Limit Acceleration=false)
   &ScaleColor(
    Scale Alpha=&FloatFromCurve(FloatCurve=DataInterface() {propertyValues {Curve {defaultValue=1, keys [{interpMode=RCIM_Cubic, tangentMode=RCTM_User, time=0.5, value=1}, {tangentMode=RCTM_User, time=1, value=0}]}}})
    ScaleA=true
    ScaleRGB=true
   )
   &SubUVAnimation(
    Evaluation Type=::NewEnumerator0
    Output Alternate Renderer Binding=true
    "Sprite Renderer "=DataInterface() {propertyValues {SpriteRenderer=&self:NE_VC_SingleBoost_Sprite_Parent003_0.NiagaraSpriteRendererProperties_1}}
    SubUV Animation Mode:ENiagara_SubUVLookupModeV2=::NewEnumerator2
    UseEndFrame=false
    UseStartFrame=false
   )
  ]
  rendererClasses [&NiagaraSpriteRendererProperties, &NiagaraMeshRendererProperties]
  renderers [
   {rendererClass=&NiagaraSpriteRendererProperties}
   {rendererClass=&NiagaraMeshRendererProperties, rendererIndex=1}
   {rendererClass=&NiagaraSpriteRendererProperties, rendererIndex=2}
   {rendererClass=&NiagaraMeshRendererProperties, rendererIndex=3}
   {rendererClass=&NiagaraMeshRendererProperties, rendererIndex=4}
   {rendererClass=&NiagaraMeshRendererProperties, rendererIndex=5}
  ]
 }
}
summary {
 systemName=NS_Pulse
 emitters [
  {emitterName=NE_VC_SingleBoost_Ribbon_Parent, rendererClasses [&NiagaraRibbonRendererProperties]}
  {
   emitterName=NE_VC_SingleBoost_Sprite_Parent001
   rendererClasses [&NiagaraSpriteRendererProperties, &NiagaraMeshRendererProperties]
  }
  {
   emitterName=NE_VC_SingleBoost_Sprite_Parent003
   rendererClasses [&NiagaraSpriteRendererProperties, &NiagaraMeshRendererProperties]
  }
  {emitterName=NE_VC_SingleBoost_Sprite_Parent, rendererClasses [&NiagaraSpriteRendererProperties]}
  {bEnabled=false, emitterName=NE_VC_SingleBoost_Sprite_Parent002, rendererClasses [&NiagaraSpriteRendererProperties]}
 ]
 userVariables [
  User.bIsFrontendPreview {type=NiagaraBool}
  User.bIsPainted {type=NiagaraBool}
  User.ColorVariantID {type=NiagaraInt32}
  User.DTCarOffset {type=NiagaraPosition, description="", defaultValue {}}
  User.GameModeID {type=NiagaraInt32}
  User.PaintedColor {type=LinearColor, description="", defaultValue {struct=LinearColor, value=<rgba 1 1 1 1>}}
  User.PaintedScalar {type=NiagaraFloat, defaultValue=1, description=""}
  User.PortCount {type=NiagaraInt32, defaultValue=1}
  User.SingleBoostPos {type=NiagaraPosition, description="", defaultValue {}}
 ]
}
