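# Physics asset for the animated Pulse boost mesh, a vehicle boost cosmetic.
# It gives seven bones one capsule body each: a large central PlateBone, five surrounding bones and a smaller ball-end bone.
# Every other body is constrained directly to PlateBone in a star layout, with identical swing and twist limit angles on each joint.
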PhysicsAsset /VehicleCosmetics/Boosts/Boost_Pulse/Meshes/SM_DLR_PulseMeshAnimated_PhysicsAsset
bodies {
 Bone_001 [{shapeType=Capsule, length=48.2736, radius=48.2142, center=<0.0114 -38.2651 0.0984>, rotation=<rot -0.0677 -90 4.152>}]
 Bone_1 [{shapeType=Capsule, length=50.4939, radius=47.0801, center=<-8.1655 30.0066 2.0036>, rotation=<rot 3.3147 -90 32.5263>}]
 Bone_2 [{shapeType=Capsule, length=50.5116, radius=45.7421, center=<-4.6521 23.6493 10.8734>, rotation=<rot 0 -72.9774 -56.5517>}]
 Bone_3 [{shapeType=Capsule, length=50.5095, radius=48.2633, center=<3.2797 27.9205 5.5506>, rotation=<rot -10.3891 -90 30.7562>}]
 Bone_4 [{shapeType=Capsule, length=50.4979, radius=48.6315, center=<-2.6829 26.071 -5.5555>, rotation=<rot 7.913 -90 -15.9162>}]
 Bone_Ball_end [{shapeType=Capsule, length=29.1997, radius=29.1288, center=<0.0808 0.2039 0.0213>, rotation=<rot -0.0004 -90 0.0056>}]
 PlateBone [{shapeType=Capsule, length=72.9117, radius=71.7092, center=<0 -44.3358 0>, rotation=<rot 0 -90 0>}]
}
constraints [
 {bone1Name=Bone_001, bone2Name=PlateBone, swing1LimitDegrees=45, swing2LimitDegrees=45, twistLimitDegrees=45}
 {bone1Name=Bone_1, bone2Name=PlateBone, swing1LimitDegrees=45, swing2LimitDegrees=45, twistLimitDegrees=45}
 {bone1Name=Bone_2, bone2Name=PlateBone, swing1LimitDegrees=45, swing2LimitDegrees=45, twistLimitDegrees=45}
 {bone1Name=Bone_4, bone2Name=PlateBone, swing1LimitDegrees=45, swing2LimitDegrees=45, twistLimitDegrees=45}
 {bone1Name=Bone_3, bone2Name=PlateBone, swing1LimitDegrees=45, swing2LimitDegrees=45, twistLimitDegrees=45}
 {bone1Name=Bone_Ball_end, bone2Name=PlateBone, swing1LimitDegrees=45, swing2LimitDegrees=45, twistLimitDegrees=45}
]
